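# Looping ambient Sound Cue for pouring water in Athena's Aztec set, routed to the indoor ambient sound class.
# A quality-level root sends its first two slots to standard stream loops and its last two to "_opt" variants.
# Each branch randomly picks one of two equally weighted looping waves.
# Overridden attenuation adds distance falloff, wide stereo spread, a non-spatialized near radius and listener focus that quiets off-focus sound.
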SoundCue /Game/Athena/Sounds/Ambient/Aztec/Amb_Aztec_WaterPour_Cue
root=QualityLevel_0
nodes [
 {id=QualityLevel_0, children [Random_0,Random_0,Random_1,Random_1]}
 {id=Random_0, children [WavePlayer_2,WavePlayer_0], properties {weights [1,1]}}
 {id=Random_1, children [WavePlayer_5,WavePlayer_3], properties {weights [1,1]}}
 {id=WavePlayer_0, properties {bLooping=true, soundWaveAssetPtr=/Game/Sounds/Fort_Ambient_Looping/Water/stream_medium_B_01}}
 {id=WavePlayer_2, properties {bLooping=true, soundWaveAssetPtr=/Game/Sounds/Fort_Ambient_Looping/Water/stream_medium_A_03}}
 {id=WavePlayer_3, properties {bLooping=true, soundWaveAssetPtr=/Game/Sounds/Fort_Ambient_Looping/Water/stream_medium_B_01_opt}}
 {id=WavePlayer_5, properties {bLooping=true, soundWaveAssetPtr=/Game/Sounds/Fort_Ambient_Looping/Water/stream_medium_A_03_opt}}
]
properties {
 bOverrideAttenuation=true
 maxDistance=2000
 soundClassObject=&Fort_Ambient_Indoors_Athena
 volumeMultiplier=0.501
 attenuationOverrides {
  bEnableFocusInterpolation=true
  bEnableListenerFocus=true
  falloffDistance=2000
  focusAzimuth=25
  nonFocusAzimuth=45
  nonFocusVolumeAttenuation=0.5
  nonSpatializedRadiusStart=300
  stereoSpread=1000
  attenuationShapeExtents {x=0}
 }
 concurrencyOverrides {maxCount=1, resolutionRule=PreventNew}
}
